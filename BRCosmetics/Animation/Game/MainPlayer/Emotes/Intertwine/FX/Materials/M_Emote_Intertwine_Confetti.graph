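# Masked, unlit Niagara sprite material for confetti, filed under the Intertwine emote's effects.
# The green channel of a 4x4 SubUV confetti normal texture is scaled and inverted to act as a shading term multiplied by particle color.
# That product feeds the emissive adjustments function, faking surface shading without lighting.
# A second SubUV confetti texture is also sampled, with no visible connection.
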Material /BRCosmetics/Animation/Game/MainPlayer/Emotes/Intertwine/FX/Materials/M_Emote_Intertwine_Confetti
graph {
 properties {bUsedWithNiagaraSprites=true, blendMode=BLEND_Masked, pixelDepthOffsetMode=PDOM_Legacy, shadingModel=MSM_Unlit, usedShadingModels=Unlit}
 nodes [
  n1 = Multiply(A=TextureSample() {samplerType=SAMPLERTYPE_Normal, texture=/Game/Effects/Fort_Effects/Textures/Confetti/T_Confetti_SubUV_4X4_03}.G) {constB=0.5}
  MaterialFunctionCall(Emissive=Multiply(A=OneMinus(None=@n1), B=ParticleColor().RGB)) {materialFunction=&MF_FX_EmissiveAdjustments}
  TextureSample() {texture=/Game/Effects/Fort_Effects/Textures/Confetti/T_Confetti_SubUV_4X4_A}
 ]
}
