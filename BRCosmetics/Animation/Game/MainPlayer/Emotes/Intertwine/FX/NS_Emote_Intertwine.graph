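# Niagara system for the Intertwine emote, layering confetti, foot dust puffs, flashing and soft sparkles, sparkle rain and a spotlight around the player.
# System bar and beat counters time confetti and rain bursts, while per-foot flags gate puffs offset to skeletal mesh samples.
# User color arrays tint confetti, sparkles and puffs, with confetti brightness following time of day.
# GPU rain particles seed a trail emitter, alongside a CPU low variant.
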NiagaraSystem /BRCosmetics/Animation/Game/MainPlayer/Emotes/Intertwine/FX/NS_Emote_Intertwine
dependencies {
 usedDataInterfaces [
  &NiagaraDataInterfaceArrayFloat
  &NiagaraDataInterfaceVectorCurve
  &NiagaraDataInterfaceCurve
  /Script/Niagara.NiagaraDataInterfaceVectorField
  &NiagaraDataInterfaceParticleRead
  &NiagaraDataInterfaceSpriteRendererInfo
  &NiagaraDataInterfaceArrayInt32
  &NiagaraDataInterfaceSkeletalMesh
 ]
 usedDynamicInputs [
  &Divide_Float
  /Niagara/DynamicInputs/Divide/Divide_Int
  &ModuloInt
  &SetBoolByFloatComparison
  &MakeFloatFromVector
 ]
 usedModules [
  &self:SystemSpawnScript.SystemScriptSource.SystemScriptGraph.NiagaraNodeAssignment_3.SetVariables_#1
  &SystemState
  &self:SystemSpawnScript.SystemScriptSource.SystemScriptGraph.NiagaraNodeAssignment_0.SetVariables_#2
  &self:SystemSpawnScript.SystemScriptSource.SystemScriptGraph.NiagaraNodeAssignment_1.SetVariables_#3
  &self:SystemSpawnScript.SystemScriptSource.SystemScriptGraph.NiagaraNodeAssignment_2.SetVariables_#4
  &self:SystemSpawnScript.SystemScriptSource.SystemScriptGraph.NiagaraNodeAssignment_5.SetVariables_#5
 ]
}
emitters {
 ConfettiSquares {
  emitterUpdateScript [
   &EmitterState()
   &SpawnBurst_Instantaneous(
    Spawn Count=25
    Spawn Probability=&MakeCustomFloatFromBool(
     Bool=&Boolean_AndOperation(
      Bool1=&SetBoolByIntComparison(A=$System.BarModulo, B=3, Comparison Type=::NewEnumerator4)
      Bool2=&SetBoolByIntComparison(A=$System.BeatModulo, B=0, Comparison Type=::NewEnumerator4)
     )
     True Float=1
     False Float=0
    )
    Use Spawn Probability=true
    UseLoopCountLimit=false
   )
  ]
  particleSpawnScript [
   &InitializeParticle(
    Color=&SelectLinearColorFromArray(Array Sampling Mode=::NewEnumerator0, Direct Array Index=?, Color Selection Array=$User.ConfettiColorArray, Random Seed=0)
    Lifetime Max=3.333
    Lifetime Mode=::NewEnumerator1
    Sprite Size Mode=::NewEnumerator1
    Uniform Sprite Size Max=6
    Uniform Sprite Size Min=3
    UsePositionOffset=false
    Write Material Random=true
   )
   &AddVelocity(
    Cone Angle=60
    Cone Axis=<0 0.5 1>
    Module.UseConeFalloff=true
    Module.UseVelDistribution=true
    Override Seed=false
    Velocity Mode=::NewEnumerator2
    Velocity Speed=&RandomRangeFloat(Minimum=666, Maximum=1111)
   )
  ]
  particleUpdateScript [
   &ParticleState()
   &GravityForce(Gravity=<0 0 -111>)
   /Niagara/Modules/Update/Forces/CurlNoiseForce(Noise Frequency=0.02, Noise Strength=200, Pan Noise=false)
   &Drag(
    Drag=&FloatFromCurve(FloatCurve=DataInterface() {propertyValues {Curve {keys [<at 0 0.5>,<at 1 1>]}}}, Scale Curve=8)
    Use Linear Drag=true
    Use Rotational Drag=true
   )
   &SolveForcesAndVelocity(Clamp Velocity=false, Limit Acceleration=false)
   &ScaleSpriteSize(Scale Factor=&Vector2DFromFloat(Value=&FloatFromCurve(FloatCurve=DataInterface() {propertyValues {Curve {keys [
          <at 0 0 RCIM_Cubic>
          {arriveTangent=-0.0623, interpMode=RCIM_Cubic, leaveTangent=-0.0623, tangentMode=RCTM_User, time=0.04, value=1}
          {arriveTangent=-0.0083, interpMode=RCIM_Cubic, leaveTangent=-0.0083, tangentMode=RCTM_User, time=0.5, value=1}
          <at 1 0 RCIM_Cubic>
         ]}}})))
   &SubUVAnimation(
    "Sprite Renderer "=DataInterface() {propertyValues {SpriteRenderer=&self:ConfettiSquares_0.NiagaraSpriteRendererProperties_0}}
    Start Frame Offset:NiagaraInt32=&RandomRangeInt(Minimum=0, Maximum=555)
    UseEndFrame=false
    UseStartFrame=false
   )
   &ScaleColor(
    Scale RGB=&VectorFromFloat(Value=&FNTimeOfDayScalar(Night Scale=0.25, Morning Scale=0.333, Day Scale=0.5, Evening Scale=0.25))
    ScaleA=true
    ScaleRGB=true
   )
  ]
  rendererClasses [&NiagaraSpriteRendererProperties]
  renderers [{rendererClass=&NiagaraSpriteRendererProperties}]
 }
 ConfettiStreamers {
  emitterUpdateScript [
   &EmitterState()
   &SpawnBurst_Instantaneous(
    Spawn Count=7
    Spawn Probability=&MakeCustomFloatFromBool(
     Bool=&Boolean_AndOperation(
      Bool1=&SetBoolByIntComparison(A=$System.BarModulo, B=3, Comparison Type=::NewEnumerator4)
      Bool2=&SetBoolByIntComparison(A=$System.BeatModulo, B=0, Comparison Type=::NewEnumerator4)
     )
     True Float=1
     False Float=0
    )
    Use Spawn Probability=true
    UseLoopCountLimit=false
   )
  ]
  particleSpawnScript [
   &InitializeParticle(
    Color=&SelectLinearColorFromArray(Array Sampling Mode=::NewEnumerator0, Direct Array Index=?, Color Selection Array=$User.ConfettiColorArray, Random Seed=0)
    Lifetime=&RandomRangeFloat(Minimum=1, Maximum=2)
    UsePositionOffset=false
    Write Material Random=true
   )
   &AddVelocity(
    Cone Angle=90
    Cone Axis=<0 0.5 1>
    Module.UseConeFalloff=true
    Module.UseVelDistribution=true
    Override Seed=false
    Velocity Mode=::NewEnumerator2
    Velocity Speed=&RandomRangeFloat(Minimum=444, Maximum=888)
   )
   &InitialMeshOrientation(
    Mesh Orientation Mode=::NewEnumerator2
    Orientation Coordinate Space=::Local
    Orientation Vector=$Particles.Velocity
    Rotation=<0.125 0 0>
    Use Rotation Vector=true
   )
  ]
  particleUpdateScript [
   &ParticleState()
   &GravityForce(Gravity=<0 0 -111>)
   &Drag(Drag=9, Use Linear Drag=true, Use Rotational Drag=false)
   &SolveForcesAndVelocity(Clamp Velocity=false, Limit Acceleration=false)
   &ScaleColor(
    Scale RGB=&VectorFromFloat(Value=&FNTimeOfDayScalar(Night Scale=0.3, Morning Scale=0.333, Day Scale=0.666, Evening Scale=0.333))
    ScaleA=false
    ScaleRGB=true
   )
  ]
  rendererClasses [&NiagaraMeshRendererProperties]
  renderers [{rendererClass=&NiagaraMeshRendererProperties, rendererIndex=0}]
 }
 DustPuffLeftFoot {
  emitterUpdateScript [
   &EmitterState(Life Cycle Mode=::NewEnumerator1, Loop Behavior=::NewEnumerator1, Loop Delay=0.1, Loop Duration Mode=::NewEnumerator1, UseLoopDelay=false)
   /SetVariables(Emitter.Age:NiagaraFloat=&MakeCustomFloatFromBool(Bool=$System.PuffCanSpawnLeftFoot, True Float=$Emitter.Age, False Float=0)) {isSetParameters=true, module=SetVariables_#6}
   &SpawnBurst_Instantaneous(
    Age:NiagaraFloat=$Emitter.Age
    Spawn Count=5
    Spawn Probability=&ClampFloat(Float=&MakeFloatFromInt(Int=$System.LoopCount))
    Use Spawn Probability=true
    UseLoopCountLimit=false
   )
  ]
  particleSpawnScript [
   &InitializeParticle(
    Color Mode=::NewEnumerator0
    Lifetime Max=0.5
    Lifetime Min=0.4
    Lifetime Mode=::NewEnumerator1
    Sprite Rotation Mode=::NewEnumerator2
    Sprite Size Mode=::NewEnumerator1
    Uniform Sprite Size Max=44
    Uniform Sprite Size Min=22
    UsePositionOffset=false
    Write Material Random=true
   )
   &ShapeLocation(
    MaskSpawnGroup=false
    Override Seed=false
    Ring / Disc Distribution Mode:ENiagaraSphereDistributionMode=::NewEnumerator1
    Ring Radius=7
    Shape Primitive=::NewEnumerator4
    U Position=&ReturnNormalizedExecIndex()
    UseOwnerScale=true
   )
   &AddVelocity(Module.Use Falloff Distance=false, Override Seed=false, Velocity Mode=::NewEnumerator1, Velocity Speed=44)
   &CameraOffset(Camera Offset Amount=33)
  ]
  particleUpdateScript [
   &SampleSkeletalMesh(
    AdditionalYawPitchRoll=false
    Bone Sampling Mode:ENiagaraBoneSamplingMode=::NewEnumerator4
    Skeletal Mesh:NiagaraDataInterfaceSkeletalMesh=$System.Skeletal Mesh
   )
   &Color(
    Color=&LerpLinearColors(
     StartColor=<rgba 1 1 1 0.5>
     EndColor=&SelectLinearColorFromArray(
      Array Sampling Mode=::NewEnumerator1
      Direct Array Index=$System.Bar
      Direct Array Mode=::NewEnumerator1
      Color Selection Array=$User.SpotlightColorArray
      Random Seed=?
     )
     LerpFactor=0.75
    )
    Scale Alpha=&Multiply_Float(
     A=0.1
     B=&SelectFloatFromArray(
      Direct Array Index=$System.LoopCount
      Direct Array Mode=::NewEnumerator1
      Float Selection Array=DataInterface() {propertyValues {FloatData [0.7,0.4,0.9,0.4,1]}}
     )
    )
    Scale Color=&VectorFromFloat(Value=0.5)
   )
   &Drag(Drag=3, Use Linear Drag=true, Use Rotational Drag=true)
   &GravityForce(Gravity=<0 0 111>)
   &ParticleState()
   &SolveForcesAndVelocity(Clamp Velocity=false, Limit Acceleration=false)
   &ScaleColor(
    Color Value To Scale=$Particles.Color
    Scale Alpha=&Multiply_Float(
     A=&FloatFromCurve(FloatCurve=DataInterface() {propertyValues {Curve {keys [
          {arriveTangent=9.0725, interpMode=RCIM_Cubic, leaveTangent=9.0725, tangentMode=RCTM_User, time=0, value=0}
          {arriveTangent=-2.5249, interpMode=RCIM_Cubic, leaveTangent=-2.5249, tangentMode=RCTM_User, time=0.1408, value=0.9847}
          {arriveTangent=-0.0405, interpMode=RCIM_Cubic, leaveTangent=-0.0405, tangentMode=RCTM_User, time=1, value=0}
         ]}}})
     B=1
    )
    Scale RGB=&VectorFromFloat(Value=&FloatFromCurve(FloatCurve=DataInterface() {propertyValues {Curve {keys [<at 0 1>,<at 0.9983 0.5>]}}}))
    ScaleA=true
    ScaleRGB=true
   )
   &DynamicMaterialParameters(
    Index 0 Param 1=&Add_Float(
     A=$Particles.RandomFloat
     B=&FloatFromCurve(
      FloatCurve=DataInterface() {propertyValues {Curve {keys [
          {arriveTangent=0.9088, interpMode=RCIM_Cubic, leaveTangent=0.9088, tangentMode=RCTM_User, time=0, value=1}
          {arriveTangent=0.5473, interpMode=RCIM_Cubic, leaveTangent=0.5473, tangentMode=RCTM_User, time=1, value=2}
         ]}}}
      CurveIndex=$Particles.Age
      Scale Curve=0.222
     )
    )
    Index 0 Param 2=&FloatFromCurve(FloatCurve=DataInterface() {propertyValues {Curve {keys [
         {arriveTangent=0.0453, interpMode=RCIM_Cubic, tangentMode=RCTM_Break, time=0, value=-1}
         {arriveTangent=0.0604, interpMode=RCIM_Cubic, leaveTangent=0.0604, tangentMode=RCTM_User, time=1, value=-1}
        ]}}})
    Index 0 Param 3=&FloatFromCurve(FloatCurve=DataInterface() {propertyValues {Curve {keys [
         {arriveTangent=-9.0447, interpMode=RCIM_Cubic, leaveTangent=-2.0044, tangentMode=RCTM_User, time=0, value=0.5}
         {interpMode=RCIM_Cubic, leaveTangent=4.6779, time=0.2851, value=0}
        ]}}})
    Index 1 Param 1=&FloatFromCurve(
     FloatCurve=DataInterface() {propertyValues {Curve {keys [<at 0 0>,<at 0.25 1>]}}}
     Scale Curve=&RandomRangeFloat(Minimum=2, Maximum=1)
    )
    Index 1 Param 2=&FloatFromCurve(
     FloatCurve=DataInterface() {propertyValues {Curve {keys [
         {interpMode=RCIM_Cubic, leaveTangent=3.7422, tangentMode=RCTM_Break, time=0, value=0.2}
         {interpMode=RCIM_Cubic, leaveTangent=0.9066, time=1, value=2}
        ]}}}
     Scale Curve=&RandomRangeFloat(Minimum=0.5, Maximum=1)
    )
    Index 1 Param 3=&FloatFromCurve(FloatCurve=DataInterface() {propertyValues {Curve {keys [
         {interpMode=RCIM_Cubic, leaveTangent=2.8445, tangentMode=RCTM_Break, time=0, value=0.2}
         {interpMode=RCIM_Cubic, leaveTangent=63.2978, time=0.25, value=1}
        ]}}})
    Index 1 Param 4=&FloatFromCurve(
     FloatCurve=DataInterface() {propertyValues {Curve {keys [<at 0 1>,<at 1 0.8>]}}}
     Scale Curve=&RandomRangeFloat(Minimum=0.5, Maximum=1)
    )
    Index 2 Param 1=&FloatFromCurve(
     FloatCurve=DataInterface() {propertyValues {Curve {keys [{interpMode=RCIM_Cubic, leaveTangent=3.2309, tangentMode=RCTM_Break, time=0.1, value=0}, <at 1 1>]}}}
     Scale Curve=0
    )
    Index 2 Param 2=&FloatFromCurve(
     FloatCurve=DataInterface() {propertyValues {Curve {keys [<at 0 0.2 RCIM_Cubic>,<at 1 0.8 RCIM_Cubic>]}}}
     Scale Curve=&RandomRangeFloat(Minimum=0.2, Maximum=1.2)
    )
    Index 2 Param 3=&FloatFromCurve(FloatCurve=DataInterface() {propertyValues {Curve {keys [
         {interpMode=RCIM_Cubic, tangentMode=RCTM_Break, time=0, value=1}
         {interpMode=RCIM_Cubic, leaveTangent=11.1233, tangentMode=RCTM_Break, time=1, value=2}
        ]}}})
    Index 2 Param 4=&FloatFromCurve(
     FloatCurve=DataInterface() {propertyValues {Curve {keys [<at 0 1 RCIM_Cubic>, {interpMode=RCIM_Cubic, leaveTangent=-1.2956, tangentMode=RCTM_Break, time=1, value=0.75}]}}}
     Scale Curve=0.8
    )
    Module.Param0WriteEnabled=true
    Module.Param0WriteEnabled1=true
    Module.Param0WriteEnabled2=true
    Module.Param1WriteEnabled=false
    Module.Param1WriteEnabled1=true
    Module.Param1WriteEnabled2=true
    Module.Param2WriteEnabled=true
    Module.Param2WriteEnabled1=true
    Module.Param2WriteEnabled2=true
    Module.Param3WriteEnabled=true
    Module.Param3WriteEnabled1=true
    Module.Param3WriteEnabled2=true
    Write Parameter Index 1=true
    Write Parameter Index 2=true
   )
   &ScaleSpriteSize(Scale Factor=&Vector2DFromFloat(Value=&FloatFromCurve(FloatCurve=DataInterface() {propertyValues {Curve {keys [
          {arriveTangent=4.1785, interpMode=RCIM_Cubic, leaveTangent=4.1785, tangentMode=RCTM_User, time=0, value=0}
          {arriveTangent=0.9995, interpMode=RCIM_Cubic, leaveTangent=0.9995, tangentMode=RCTM_User, time=0.2223, value=0.6825}
          {arriveTangent=-0.0405, interpMode=RCIM_Cubic, leaveTangent=-0.0405, tangentMode=RCTM_User, time=1, value=1}
         ]}}})))
   &SpriteRotationRate(Rotation Rate=&Multiply_Float(
     A=$Particles.InitSpriteRotRate
     B=&FloatFromCurve(
      FloatCurve=DataInterface() {propertyValues {Curve {keys [
          {arriveTangent=-1.9214, interpMode=RCIM_Cubic, leaveTangent=-1.9214, tangentMode=RCTM_User, time=-0.0046, value=1.0113}
          {arriveTangent=-0.2496, interpMode=RCIM_Cubic, leaveTangent=-0.2496, tangentMode=RCTM_User, time=0.9896, value=0.16}
         ]}}}
      CurveIndex=$Emitter.Age
      Scale Curve=150
     )
    ))
   /SetVariables(Particles.OffsetPosition:NiagaraPosition=&AddVectorToPosition(Position=$Particles.Position, Vector=$Particles.SampleSkeletalMesh.SampledPosition)) {isSetParameters=true, module=SetVariables_#7}
  ]
  rendererClasses [&NiagaraSpriteRendererProperties]
  renderers [{rendererClass=&NiagaraSpriteRendererProperties}]
 }
 DustPuffRightFoot {
  emitterUpdateScript [
   &EmitterState(Life Cycle Mode=::NewEnumerator1, Loop Behavior=::NewEnumerator1, Loop Delay=0.1, Loop Duration Mode=::NewEnumerator1, UseLoopDelay=false)
   /SetVariables(Emitter.Age:NiagaraFloat=&MakeCustomFloatFromBool(Bool=$System.PuffCanSpawnRightFoot, True Float=$Emitter.Age, False Float=0)) {isSetParameters=true, module=SetVariables_#6}
   &SpawnBurst_Instantaneous(
    Age:NiagaraFloat=$Emitter.Age
    Spawn Count=5
    Spawn Probability=&ClampFloat(Float=&MakeFloatFromInt(Int=$System.LoopCount))
    Use Spawn Probability=true
    UseLoopCountLimit=false
   )
  ]
  particleSpawnScript [
   &InitializeParticle(
    Color Mode=::NewEnumerator0
    Lifetime Max=0.5
    Lifetime Min=0.4
    Lifetime Mode=::NewEnumerator1
    Sprite Rotation Mode=::NewEnumerator2
    Sprite Size Mode=::NewEnumerator1
    Uniform Sprite Size Max=44
    Uniform Sprite Size Min=22
    UsePositionOffset=false
    Write Material Random=true
   )
   &ShapeLocation(
    MaskSpawnGroup=false
    Override Seed=false
    Ring / Disc Distribution Mode:ENiagaraSphereDistributionMode=::NewEnumerator1
    Ring Radius=7
    Shape Primitive=::NewEnumerator4
    U Position=&ReturnNormalizedExecIndex()
    UseOwnerScale=true
   )
   &AddVelocity(Module.Use Falloff Distance=false, Override Seed=false, Velocity Mode=::NewEnumerator1, Velocity Speed=44)
   &CameraOffset(Camera Offset Amount=33)
  ]
  particleUpdateScript [
   &SampleSkeletalMesh(
    AdditionalYawPitchRoll=false
    Bone / Socket Index=1
    Bone Sampling Mode:ENiagaraBoneSamplingMode=::NewEnumerator4
    Skeletal Mesh:NiagaraDataInterfaceSkeletalMesh=$System.Skeletal Mesh
   )
   &Color(
    Color=&LerpLinearColors(
     StartColor=<rgba 1 1 1 1>
     EndColor=&SelectLinearColorFromArray(
      Array Sampling Mode=::NewEnumerator1
      Direct Array Index=$System.Bar
      Direct Array Mode=::NewEnumerator1
      Color Selection Array=$User.SpotlightColorArray
      Random Seed=?
     )
     LerpFactor=0.75
    )
    Scale Alpha=&Multiply_Float(
     A=0.1
     B=&SelectFloatFromArray(
      Direct Array Index=$System.LoopCount
      Direct Array Mode=::NewEnumerator1
      Float Selection Array=DataInterface() {propertyValues {FloatData [0.7,0.4,0.9,0.4,1]}}
     )
    )
    Scale Color=&VectorFromFloat(Value=0.5)
   )
   &Drag(Drag=3, Use Linear Drag=true, Use Rotational Drag=true)
   &GravityForce(Gravity=<0 0 111>)
   &ParticleState()
   &SolveForcesAndVelocity(Clamp Velocity=false, Limit Acceleration=false)
   &ScaleColor(
    Color Value To Scale=$Particles.Color
    Scale Alpha=&Multiply_Float(
     A=&FloatFromCurve(FloatCurve=DataInterface() {propertyValues {Curve {keys [
          {arriveTangent=9.0725, interpMode=RCIM_Cubic, leaveTangent=9.0725, tangentMode=RCTM_User, time=0, value=0}
          {arriveTangent=-2.5249, interpMode=RCIM_Cubic, leaveTangent=-2.5249, tangentMode=RCTM_User, time=0.1408, value=0.9847}
          {arriveTangent=-0.0405, interpMode=RCIM_Cubic, leaveTangent=-0.0405, tangentMode=RCTM_User, time=1, value=0}
         ]}}})
     B=1
    )
    Scale RGB=&VectorFromFloat(Value=&FloatFromCurve(FloatCurve=DataInterface() {propertyValues {Curve {keys [<at 0 1>,<at 0.9983 0.5>]}}}))
    ScaleA=true
    ScaleRGB=true
   )
   &DynamicMaterialParameters(
    Index 0 Param 1=&Add_Float(
     A=$Particles.RandomFloat
     B=&FloatFromCurve(
      FloatCurve=DataInterface() {propertyValues {Curve {keys [
          {arriveTangent=0.9088, interpMode=RCIM_Cubic, leaveTangent=0.9088, tangentMode=RCTM_User, time=0, value=1}
          {arriveTangent=0.5473, interpMode=RCIM_Cubic, leaveTangent=0.5473, tangentMode=RCTM_User, time=1, value=2}
         ]}}}
      CurveIndex=$Particles.Age
      Scale Curve=0.222
     )
    )
    Index 0 Param 2=&FloatFromCurve(FloatCurve=DataInterface() {propertyValues {Curve {keys [
         {arriveTangent=0.0453, interpMode=RCIM_Cubic, tangentMode=RCTM_Break, time=0, value=-1}
         {arriveTangent=0.0604, interpMode=RCIM_Cubic, leaveTangent=0.0604, tangentMode=RCTM_User, time=1, value=-1}
        ]}}})
    Index 0 Param 3=&FloatFromCurve(FloatCurve=DataInterface() {propertyValues {Curve {keys [
         {arriveTangent=-9.0447, interpMode=RCIM_Cubic, leaveTangent=-2.0044, tangentMode=RCTM_User, time=0, value=0.5}
         {interpMode=RCIM_Cubic, leaveTangent=4.6779, time=0.2851, value=0}
        ]}}})
    Index 1 Param 1=&FloatFromCurve(
     FloatCurve=DataInterface() {propertyValues {Curve {keys [<at 0 0>,<at 0.25 1>]}}}
     Scale Curve=&RandomRangeFloat(Minimum=2, Maximum=1)
    )
    Index 1 Param 2=&FloatFromCurve(
     FloatCurve=DataInterface() {propertyValues {Curve {keys [
         {interpMode=RCIM_Cubic, leaveTangent=3.7422, tangentMode=RCTM_Break, time=0, value=0.2}
         {interpMode=RCIM_Cubic, leaveTangent=0.9066, time=1, value=2}
        ]}}}
     Scale Curve=&RandomRangeFloat(Minimum=0.5, Maximum=1)
    )
    Index 1 Param 3=&FloatFromCurve(FloatCurve=DataInterface() {propertyValues {Curve {keys [
         {interpMode=RCIM_Cubic, leaveTangent=2.8445, tangentMode=RCTM_Break, time=0, value=0.2}
         {interpMode=RCIM_Cubic, leaveTangent=63.2978, time=0.25, value=1}
        ]}}})
    Index 1 Param 4=&FloatFromCurve(
     FloatCurve=DataInterface() {propertyValues {Curve {keys [<at 0 1>,<at 1 0.8>]}}}
     Scale Curve=&RandomRangeFloat(Minimum=0.5, Maximum=1)
    )
    Index 2 Param 1=&FloatFromCurve(
     FloatCurve=DataInterface() {propertyValues {Curve {keys [{interpMode=RCIM_Cubic, leaveTangent=3.2309, tangentMode=RCTM_Break, time=0.1, value=0}, <at 1 1>]}}}
     Scale Curve=0
    )
    Index 2 Param 2=&FloatFromCurve(
     FloatCurve=DataInterface() {propertyValues {Curve {keys [<at 0 0.2 RCIM_Cubic>,<at 1 0.8 RCIM_Cubic>]}}}
     Scale Curve=&RandomRangeFloat(Minimum=0.2, Maximum=1.2)
    )
    Index 2 Param 3=&FloatFromCurve(FloatCurve=DataInterface() {propertyValues {Curve {keys [
         {interpMode=RCIM_Cubic, tangentMode=RCTM_Break, time=0, value=1}
         {interpMode=RCIM_Cubic, leaveTangent=11.1233, tangentMode=RCTM_Break, time=1, value=2}
        ]}}})
    Index 2 Param 4=&FloatFromCurve(
     FloatCurve=DataInterface() {propertyValues {Curve {keys [<at 0 1 RCIM_Cubic>, {interpMode=RCIM_Cubic, leaveTangent=-1.2956, tangentMode=RCTM_Break, time=1, value=0.75}]}}}
     Scale Curve=0.8
    )
    Module.Param0WriteEnabled=true
    Module.Param0WriteEnabled1=true
    Module.Param0WriteEnabled2=true
    Module.Param1WriteEnabled=false
    Module.Param1WriteEnabled1=true
    Module.Param1WriteEnabled2=true
    Module.Param2WriteEnabled=true
    Module.Param2WriteEnabled1=true
    Module.Param2WriteEnabled2=true
    Module.Param3WriteEnabled=true
    Module.Param3WriteEnabled1=true
    Module.Param3WriteEnabled2=true
    Write Parameter Index 1=true
    Write Parameter Index 2=true
   )
   &ScaleSpriteSize(Scale Factor=&Vector2DFromFloat(Value=&FloatFromCurve(FloatCurve=DataInterface() {propertyValues {Curve {keys [
          {arriveTangent=4.1785, interpMode=RCIM_Cubic, leaveTangent=4.1785, tangentMode=RCTM_User, time=0, value=0}
          {arriveTangent=0.9995, interpMode=RCIM_Cubic, leaveTangent=0.9995, tangentMode=RCTM_User, time=0.2223, value=0.6825}
          {arriveTangent=-0.0405, interpMode=RCIM_Cubic, leaveTangent=-0.0405, tangentMode=RCTM_User, time=1, value=1}
         ]}}})))
   &SpriteRotationRate(Rotation Rate=&Multiply_Float(
     A=$Particles.InitSpriteRotRate
     B=&FloatFromCurve(
      FloatCurve=DataInterface() {propertyValues {Curve {keys [
          {arriveTangent=-1.9214, interpMode=RCIM_Cubic, leaveTangent=-1.9214, tangentMode=RCTM_User, time=-0.0046, value=1.0113}
          {arriveTangent=-0.2496, interpMode=RCIM_Cubic, leaveTangent=-0.2496, tangentMode=RCTM_User, time=0.9896, value=0.16}
         ]}}}
      CurveIndex=$Emitter.Age
      Scale Curve=150
     )
    ))
   /SetVariables(Particles.OffsetPosition:NiagaraPosition=&AddVectorToPosition(Position=$Particles.Position, Vector=$Particles.SampleSkeletalMesh.SampledPosition)) {isSetParameters=true, module=SetVariables_#7}
  ]
  rendererClasses [&NiagaraSpriteRendererProperties]
  renderers [{rendererClass=&NiagaraSpriteRendererProperties}]
 }
 FlashingSparkles {
  emitterUpdateScript [
   &EmitterState()
   &SpawnRate(SpawnRate=9, Use Spawn Probability=false)
  ]
  particleSpawnScript [
   &InitializeParticle(
    Color=&SelectLinearColorFromArray(Array Sampling Mode=::NewEnumerator0, Direct Array Index=?, Color Selection Array=$User.SpotlightColorArray, Random Seed=0)
    Lifetime Max=0.15
    Lifetime Min=0.2
    Lifetime Mode=::NewEnumerator1
    Sprite Rotation Angle=&UniformAOrBFloat(A=0, B=0.125)
    Sprite Rotation Mode=::NewEnumerator1
    Sprite Size Mode=::NewEnumerator1
    Uniform Sprite Size Max=66
    Uniform Sprite Size Min=33
    UsePositionOffset=false
    Write Material Random=true
   )
   &ShapeLocation(
    MaskSpawnGroup=false
    Non Uniform Scale=<0.666 0.666 1>
    Offset=<0 0 133>
    Offset Mode=::NewEnumerator0
    Shape Origin=<0 0 0>
    Sphere Distribution=::NewEnumerator1
    Sphere Radius=99
    Sphere Radius Position:NiagaraFloat=&Lerp_Float(A=0, B=1, Alpha=&Power(a=&RandomRangeFloat(Minimum=0, Maximum=1), b=1))
    U Position=&RandomRangeFloat(Minimum=0, Maximum=1)
    UseOwnerScale=true
    V Position=&RandomRangeFloat(Minimum=0, Maximum=1)
   )
   &CameraOffset(Camera Offset Amount=-99)
  ]
  particleUpdateScript [
   &ParticleState()
   &ScaleColor(
    Scale Alpha=&FloatFromCurve(FloatCurve=DataInterface() {propertyValues {Curve {keys [
         {arriveTangent=9.8157, interpMode=RCIM_Cubic, leaveTangent=9.8157, tangentMode=RCTM_User, time=0, value=0}
         {arriveTangent=-1.4617, interpMode=RCIM_Cubic, leaveTangent=-1.4617, tangentMode=RCTM_User, time=0.2, value=1}
         {interpMode=RCIM_Cubic, tangentMode=RCTM_User, time=0.3, value=0}
         {interpMode=RCIM_Cubic, tangentMode=RCTM_User, time=0.5, value=0.75}
         {interpMode=RCIM_Cubic, tangentMode=RCTM_Break, time=1, value=0}
        ]}}})
    ScaleA=true
    ScaleRGB=true
   )
   &ScaleSpriteSize(Uniform Curve Sprite Scale=DataInterface() {propertyValues {Curve {keys [{tangentMode=RCTM_User, time=0, value=1}, {tangentMode=RCTM_User, time=1, value=0}]}}})
  ]
  rendererClasses [&NiagaraSpriteRendererProperties]
  renderers [{rendererClass=&NiagaraSpriteRendererProperties}]
 }
 SoftSparkles {
  simTarget=GPUComputeSim
  emitterUpdateScript [
   &EmitterState()
   &SpawnRate(SpawnRate=20, Use Spawn Probability=false)
  ]
  particleSpawnScript [
   &InitializeParticle(
    Color=&MultiplyLinearColor(
     Linear Color=&SelectLinearColorFromArray(Array Sampling Mode=::NewEnumerator0, Direct Array Index=?, Color Selection Array=$User.SpotlightColorArray, Random Seed=0)
     Scale Factor=&MakeVector4_FromFloat(Float=0.1)
    )
    Lifetime Max=2.4
    Lifetime Min=1.2
    Lifetime Mode=::NewEnumerator1
    Sprite Size Mode=::NewEnumerator1
    Uniform Sprite Size Max=44
    Uniform Sprite Size Min=22
    UsePositionOffset=false
    Write Material Random=true
   )
   &ShapeLocation(
    MaskSpawnGroup=false
    Non Uniform Scale=<0.666 0.666 1>
    Offset=<0 0 133>
    Offset Mode=::NewEnumerator0
    Shape Origin=<0 0 0>
    Sphere Distribution=::NewEnumerator1
    Sphere Radius=99
    Sphere Radius Position:NiagaraFloat=&Lerp_Float(A=0, B=1, Alpha=&Power(a=&RandomRangeFloat(Minimum=0, Maximum=1), b=1))
    U Position=&RandomRangeFloat(Minimum=0, Maximum=1)
    UseOwnerScale=true
    V Position=&RandomRangeFloat(Minimum=0, Maximum=1)
   )
   /SetVariables(Particles.Curve:NiagaraFloat=&FloatFromCurve(FloatCurve=DataInterface() {propertyValues {Curve {keys [
         {interpMode=RCIM_Cubic, leaveTangent=7.0743, tangentMode=RCTM_Break, time=0, value=0.35}
         {interpMode=RCIM_Cubic, tangentMode=RCTM_Break, time=0.2, value=0.95}
         {interpMode=RCIM_Cubic, leaveTangent=-0.1241, tangentMode=RCTM_Break, time=0.4, value=0.55}
         {interpMode=RCIM_Cubic, tangentMode=RCTM_Break, time=0.6, value=0.8}
         {interpMode=RCIM_Cubic, tangentMode=RCTM_Break, time=0.8, value=0.6}
         {arriveTangent=2.6513, interpMode=RCIM_Cubic, tangentMode=RCTM_Break, time=1, value=1}
        ]}}})) {isSetParameters=true}
  ]
  particleUpdateScript [
   &ParticleState()
   /Niagara/Modules/Update/Forces/V2/CurlNoiseForce(
    Noise Frequency=100
    Noise Strength=&FloatFromCurve(
     FloatCurve=DataInterface() {propertyValues {Curve {keys [{tangentMode=RCTM_User, time=0, value=0}, {tangentMode=RCTM_User, time=1, value=1}]}}}
     Scale Curve=111
    )
    Pan Noise=false
   )
   &SolveForcesAndVelocity(Clamp Velocity=false, Limit Acceleration=false)
   &DynamicMaterialParameters(
    Index 0 Param 1=&Multiply_Float(
     A=&FloatFromCurve(FloatCurve=DataInterface() {propertyValues {Curve {keys [
          {interpMode=RCIM_Cubic, leaveTangent=0.1514, tangentMode=RCTM_Break, time=0, value=1}
          {arriveTangent=-0.3559, interpMode=RCIM_Cubic, tangentMode=RCTM_Break, time=1, value=0}
         ]}}})
     B=&Lerp_Float(A=1, B=0, Alpha=$Particles.Curve)
    )
    Module.Param0WriteEnabled=true
    Module.Param1WriteEnabled=true
    Module.Param2WriteEnabled=true
    Module.Param3WriteEnabled=true
   )
   &ScaleColor(
    Scale Alpha=&FloatFromCurve(FloatCurve=DataInterface() {propertyValues {Curve {keys [
         {interpMode=RCIM_Cubic, tangentMode=RCTM_User, time=0, value=0}
         {interpMode=RCIM_Cubic, tangentMode=RCTM_User, time=0.5, value=1}
         {tangentMode=RCTM_User, time=1, value=0}
        ]}}})
    Scale RGB=&VectorFromFloat(Value=&Lerp_Float(
      A=1
      B=5
      Alpha=&Sine(Normalized Angle=$Emitter.Age, Period=&RandomRangeFloat(Minimum=0.1, Maximum=0.4), Scale=1, Bias=1)
     ))
    ScaleA=true
    ScaleRGB=true
   )
   &ScaleSpriteSize(
    Scale Sprite Size Mode=::NewEnumerator0
    Uniform Scale Factor=&Multiply_Float(
     A=&FloatFromCurve(FloatCurve=DataInterface() {propertyValues {Curve {keys [
          {interpMode=RCIM_Cubic, leaveTangent=2.1941, tangentMode=RCTM_Break, time=0, value=0.35}
          {arriveTangent=0.204, interpMode=RCIM_Cubic, tangentMode=RCTM_Break, time=1, value=1}
         ]}}})
     B=$Particles.Curve
    )
   )
  ]
  rendererClasses [&NiagaraSpriteRendererProperties]
  renderers [{rendererClass=&NiagaraSpriteRendererProperties}]
 }
 SparkleRainCPULow {
  emitterUpdateScript [
   &EmitterState(Life Cycle Mode=::NewEnumerator1, Loop Duration=$System.CurrentLoopDuration, UseLoopDelay=false)
   &SpawnBurst_Instantaneous(
    Spawn Count=22
    Spawn Probability=&MakeCustomFloatFromBool(
     Bool=&Boolean_AndOperation(
      Bool1=&SetBoolByIntComparison(A=$System.BarModulo, B=1, Comparison Type=::NewEnumerator4)
      Bool2=&SetBoolByIntComparison(A=$System.BeatModulo, B=1, Comparison Type=::NewEnumerator4)
     )
     True Float=1
     False Float=0
    )
    Use Spawn Probability=true
    UseLoopCountLimit=false
   )
  ]
  particleSpawnScript [
   &InitializeParticle(
    AdjustAlpha=false
    AdjustHue=true
    AdjustSaturation=false
    AdjustValue=true
    Color=<rgba 0 0.6667 1 1>
    Color Mode=::NewEnumerator3
    Hue Shift Range:Vector2f=<-0.05 0.05>
    Lifetime Max=1.666
    Lifetime Min=0.666
    Lifetime Mode=::NewEnumerator1
    Sprite Rotation Angle=0.125
    Sprite Rotation Mode=::NewEnumerator1
    Sprite Size Mode=::NewEnumerator1
    Uniform Sprite Size Max=33
    Uniform Sprite Size Min=16
    UsePositionOffset=false
    Value Range:Vector2f=<0.3 11>
    Write Material Random=true
   )
   &AddVelocity(
    Cone Axis=<0 0 -1>
    Module.UseConeFalloff=true
    Module.UseVelDistribution=true
    Override Seed=false
    Velocity Mode=::NewEnumerator2
    Velocity Speed=&RandomRangeFloat(Minimum=111, Maximum=333)
   )
   &SystemLocation(MaskSpawnGroup=false, Offset=<0 0 280>)
  ]
  particleUpdateScript [
   &GravityForce(Gravity=<0 0 -64>)
   &Drag(Use Linear Drag=true, Use Rotational Drag=true)
   &CurlNoiseForce(Noise Frequency=0.25, Noise Strength=15, Pan Noise=false)
   &SolveForcesAndVelocity(Clamp Velocity=false, Limit Acceleration=false)
   &ParticleState()
   &DynamicMaterialParameters(
    Index 0 Param 1=0.8
    Index 0 Param 2=1
    Index 0 Param 3=1
    Index 1 Param 1=150
    Index 1 Param 4=1
    Module.Param0WriteEnabled=true
    Module.Param0WriteEnabled1=true
    Module.Param1WriteEnabled=true
    Module.Param1WriteEnabled1=true
    Module.Param2WriteEnabled=true
    Module.Param2WriteEnabled1=true
    Module.Param3WriteEnabled=false
    Module.Param3WriteEnabled1=true
    Write Parameter Index 1=true
   )
   &ScaleColor(
    Scale Alpha=&FloatFromCurve(
     FloatCurve=DataInterface() {propertyValues {Curve {keys [{interpMode=RCIM_Cubic, tangentMode=RCTM_User, time=0, value=1}, {interpMode=RCIM_Cubic, tangentMode=RCTM_User, time=1, value=0}]}}}
     Scale Curve=&Waveform(
      Global Phase=&Add_Float(A=$Particles.Age, B=&RandomRangeFloat(Minimum=0, Maximum=1))
      "Waveform [1]"=::NewEnumerator0
      "[1] Frequency"=&RandomRangeFloat(Minimum=1.618, Maximum=3.618)
      "[1] Amplitude Min/Max"=<0 1>
     )
    )
    Scale RGB=&VectorFromFloat(Value=&Lerp_Float(
      A=1
      B=3
      Alpha=&FloatFromCurve(FloatCurve=DataInterface() {propertyValues {Curve {keys [<at 0 1>,<at 0.25 0>]}}})
     ))
    ScaleA=true
    ScaleRGB=true
   )
   &ScaleSpriteSize(Uniform Curve Sprite Scale=DataInterface() {propertyValues {Curve {keys [
        {arriveTangent=2.9868, time=0, value=0}
        {leaveTangent=0.2875, time=0.1, value=0.833}
        <at 0.9 1>
        {arriveTangent=-0.0405, leaveTangent=-0.0405, time=1, value=0}
       ]}}})
   /Niagara/VectorFields/SampleVectorField(
    FieldRotate:Quat4f=<0 0 1 1>
    Sampled Vector Scale=0.05
    VectorField=DataInterface() {propertyValues {Field=/Game/Effects/VectorFields/Noisy_Turbulence_103, bTileX=true, bTileY=true, bTileZ=true}}
   )
   /Niagara/VectorFields/ApplyVectorField(Vector Field Intensity as Force=0, Vector Field Intensity as Velocity=50, Vector Field Velocity Tightness=0.001)
  ]
  rendererClasses [&NiagaraSpriteRendererProperties]
  renderers [{rendererClass=&NiagaraSpriteRendererProperties}]
 }
 SparkleRainParent {
  simTarget=GPUComputeSim
  emitterUpdateScript [
   &EmitterState(Life Cycle Mode=::NewEnumerator1, Loop Duration=$System.CurrentLoopDuration, UseLoopDelay=false)
   &SpawnBurst_Instantaneous(
    Spawn Count=22
    Spawn Probability=&MakeCustomFloatFromBool(
     Bool=&Boolean_AndOperation(
      Bool1=&SetBoolByIntComparison(A=$System.BarModulo, B=1, Comparison Type=::NewEnumerator4)
      Bool2=&SetBoolByIntComparison(A=$System.BeatModulo, B=1, Comparison Type=::NewEnumerator4)
     )
     True Float=1
     False Float=0
    )
    Use Spawn Probability=true
    UseLoopCountLimit=false
   )
  ]
  particleSpawnScript [
   &InitializeParticle(
    AdjustAlpha=false
    AdjustHue=true
    AdjustSaturation=false
    AdjustValue=true
    Color=<rgba 0 0.6667 1 1>
    Color Mode=::NewEnumerator3
    Hue Shift Range:Vector2f=<-0.05 0.05>
    Lifetime Max=1.666
    Lifetime Min=0.666
    Lifetime Mode=::NewEnumerator1
    Sprite Rotation Angle=0.125
    Sprite Rotation Mode=::NewEnumerator1
    Sprite Size Mode=::NewEnumerator1
    Uniform Sprite Size Max=33
    Uniform Sprite Size Min=16
    UsePositionOffset=false
    Value Range:Vector2f=<0.3 11>
    Write Material Random=true
   )
   &AddVelocity(
    Cone Axis=<0 0 -1>
    Module.UseConeFalloff=true
    Module.UseVelDistribution=true
    Override Seed=false
    Velocity Mode=::NewEnumerator2
    Velocity Speed=&RandomRangeFloat(Minimum=111, Maximum=333)
   )
   &SystemLocation(MaskSpawnGroup=false, Offset=<0 0 280>)
  ]
  particleUpdateScript [
   &GravityForce(Gravity=<0 0 -64>)
   &Drag(Use Linear Drag=true, Use Rotational Drag=true)
   &CurlNoiseForce(Noise Frequency=0.25, Noise Strength=15, Pan Noise=false)
   &SolveForcesAndVelocity(Clamp Velocity=false, Limit Acceleration=false)
   &ParticleState()
   &DynamicMaterialParameters(
    Index 0 Param 1=0.8
    Index 0 Param 2=1
    Index 0 Param 3=1
    Index 1 Param 1=150
    Index 1 Param 4=1
    Module.Param0WriteEnabled=true
    Module.Param0WriteEnabled1=true
    Module.Param1WriteEnabled=true
    Module.Param1WriteEnabled1=true
    Module.Param2WriteEnabled=true
    Module.Param2WriteEnabled1=true
    Module.Param3WriteEnabled=false
    Module.Param3WriteEnabled1=true
    Write Parameter Index 1=true
   )
   &ScaleColor(
    Scale Alpha=&FloatFromCurve(
     FloatCurve=DataInterface() {propertyValues {Curve {keys [{interpMode=RCIM_Cubic, tangentMode=RCTM_User, time=0, value=1}, {interpMode=RCIM_Cubic, tangentMode=RCTM_User, time=1, value=0}]}}}
     Scale Curve=&Waveform(
      Global Phase=&Add_Float(A=$Particles.Age, B=&RandomRangeFloat(Minimum=0, Maximum=1))
      "Waveform [1]"=::NewEnumerator0
      "[1] Frequency"=&RandomRangeFloat(Minimum=1.618, Maximum=3.618)
      "[1] Amplitude Min/Max"=<0 1>
     )
    )
    Scale RGB=&VectorFromFloat(Value=&Lerp_Float(
      A=1
      B=3
      Alpha=&FloatFromCurve(FloatCurve=DataInterface() {propertyValues {Curve {keys [<at 0 1>,<at 0.25 0>]}}})
     ))
    ScaleA=true
    ScaleRGB=true
   )
   &ScaleSpriteSize(Uniform Curve Sprite Scale=DataInterface() {propertyValues {Curve {keys [
        {arriveTangent=2.9868, time=0, value=0}
        {leaveTangent=0.2875, time=0.1, value=0.833}
        <at 0.9 1>
        {arriveTangent=-0.0405, leaveTangent=-0.0405, time=1, value=0}
       ]}}})
   &SampleVectorField(
    FieldRotate:Quat4f=<0 0 1 1>
    Sampled Vector Scale=0.05
    VectorField=DataInterface() {propertyValues {Field=&Noisy_Turbulence_103, bTileX=true, bTileY=true, bTileZ=true}}
   )
   &ApplyVectorField(Vector Field Intensity as Force=0, Vector Field Intensity as Velocity=50, Vector Field Velocity Tightness=0.001)
  ]
  rendererClasses [&NiagaraSpriteRendererProperties]
  renderers [{rendererClass=&NiagaraSpriteRendererProperties}]
 }
 SparkleRainTrail {
  simTarget=GPUComputeSim
  emitterUpdateScript [
   &EmitterState()
   &SpawnParticlesFromOtherEmitter(
    Attribute Reader=DataInterface() {propertyValues {EmitterBinding {bindingMode=Other, emitterName=SparkleRainParent}}}
    Maximum Age Spawn Threshold=2
    Minimum Age Spawn Threshold=0
    Spawn Rate=44
    Spawn Rate Per Particle Cap=22
    Use Maximum Age=false
    Use Minimum Age=false
   )
  ]
  particleSpawnScript [
   &InitializeParticle(
    Color Maximum=<rgba 0 0.1998 0.666 1>
    Color Minimum=<rgba 0 0.8667 1 1>
    Color Mode=::NewEnumerator2
    Lifetime Max=1.333
    Lifetime Min=0.333
    Lifetime Mode=::NewEnumerator1
    Sprite Rotation Angle=0.125
    Sprite Rotation Mode=::NewEnumerator1
    Sprite Size Mode=::NewEnumerator1
    Uniform Sprite Size Max=11
    Uniform Sprite Size Min=11
    UsePositionOffset=false
    Write Material Random=true
   )
   &SampleParticlesFromOtherEmitter(
    Applied Velocity Scale:NiagaraFloat=&FloatFromCurve(FloatCurve=DataInterface() {propertyValues {Curve {keys [<at 0 1>,<at 1 0>]}}}, Scale Curve=0.5)
    Maximum Sampled Particle Age=2
    Minimum Sampled Particle Age=0
    Module.MaxSampledAgeCutoff=true
    Module.MinSampledAgeCutoff=true
    Reject Particle By Sampled Age=true
    Velocity Sampling=::NewEnumerator0
   )
   &ShapeLocation(MaskSpawnGroup=false, Override Seed=false, Sphere Radius=5, UseOwnerScale=true)
  ]
  particleUpdateScript [
   &ParticleState()
   &Drag(Drag=2, Use Linear Drag=true, Use Rotational Drag=false)
   &GravityForce(Gravity=<0 0 -35>)
   &CurlNoiseForce(Noise Frequency=1, Noise Strength=15, Pan Noise=false)
   &SolveForcesAndVelocity(Clamp Velocity=false, Limit Acceleration=false)
   &ScaleColor(
    Scale Alpha=&FloatFromCurve(
     FloatCurve=DataInterface() {propertyValues {Curve {keys [
         {tangentMode=RCTM_User, time=0, value=1}
         {arriveTangent=-0.0405, interpMode=RCIM_Cubic, leaveTangent=-0.0405, tangentMode=RCTM_User, time=0.7009, value=1}
         {arriveTangent=-8.0089, interpMode=RCIM_Cubic, leaveTangent=-8.0089, tangentMode=RCTM_User, time=1, value=0}
        ]}}}
     Scale Curve=&Waveform(
      Global Phase=&Add_Float(A=$Particles.Age, B=&RandomRangeFloat(Minimum=0, Maximum=1))
      "Waveform [1]"=::NewEnumerator0
      "[1] Frequency"=&RandomRangeFloat(Minimum=3.618, Maximum=11.08)
      "[1] Amplitude Min/Max"=<0 1>
     )
    )
    Scale RGB=&VectorFromFloat(Value=&Lerp_Float(
      A=0.25
      B=22
      Alpha=&FloatFromCurve(FloatCurve=DataInterface() {propertyValues {Curve {keys [<at 0 1>,<at 0.333 0>]}}})
     ))
    ScaleA=true
    ScaleRGB=true
   )
   &ScaleSpriteSize(Uniform Curve Sprite Scale=DataInterface() {propertyValues {Curve {keys [
        {interpMode=RCIM_Cubic, tangentMode=RCTM_User, time=0, value=0}
        {interpMode=RCIM_Cubic, tangentMode=RCTM_User, time=0.5, value=1}
        {tangentMode=RCTM_User, time=1, value=0}
       ]}}})
  ]
  rendererClasses [&NiagaraSpriteRendererProperties]
  renderers [{rendererClass=&NiagaraSpriteRendererProperties}]
 }
 SpotLightMain {
  emitterUpdateScript [
   &EmitterState(
    Inactive Response=::NewEnumerator1
    Life Cycle Mode=::NewEnumerator1
    Loop Behavior=::NewEnumerator1
    Loop Duration Mode=::NewEnumerator1
    UseLoopDelay=false
   )
   &SpawnBurst_Instantaneous(Use Spawn Probability=false, UseLoopCountLimit=false)
  ]
  particleSpawnScript [
   &InitializeParticle(
    Mesh Scale=<0.8 0.8 1.2>
    Mesh Scale Mode=::NewEnumerator4
    Sprite Size Mode=::NewEnumerator3
    Uniform Sprite Size=333
    UsePositionOffset=false
    Write Material Random=true
   )
   &CameraOffset()
   /SetVariables(
    Particles.FloorGlowFacingVector:Vector3f=<0 0 1>
    Particles.FloorGlowPosition:NiagaraPosition=&AddVectorToPosition(Position=/Niagara/DynamicInputs/Helpers/SimulationPosition(), Vector=<0 0 4>)
    Particles.FloorGlowSize:Vector2f=&Vector2DFromFloat(Value=250)
    Particles.LightAttenuationScale=275
    Particles.LightExponent=2
    Particles.LightPosition:NiagaraPosition=&AddVectorToPosition(Position=&SimulationPosition(), Vector=<0 0 200>)
    Particles.LightRadius=230
    Particles.SpotLightPosition:NiagaraPosition=&AddVectorToPosition(Position=&SimulationPosition(), Vector=<0 0 280>)
   ) {isSetParameters=true, module=SetVariables_#8}
  ]
  particleUpdateScript [
   &ParticleState(
    Kill Particles When Lifetime Has Elapsed=false
    Let Infinitely Lived Particles Die When Emitter Deactivates=true
    Loop Particles Lifetime=true
   )
   &Color(
    Color=&SelectLinearColorFromArray(
     Array Sampling Mode=::NewEnumerator1
     Direct Array Index=$System.Bar
     Direct Array Mode=::NewEnumerator1
     Color Selection Array=$User.SpotlightColorArray
     Random Seed=?
    )
    Scale Alpha=&Multiply_Float(
     A=$System.Oscillation
     B=&SelectFloatFromArray(
      Direct Array Index=$System.LoopCount
      Direct Array Mode=::NewEnumerator1
      Float Selection Array=DataInterface() {propertyValues {FloatData [0.7,0.4,0.9,0.4,1]}}
     )
    )
   )
   /SetVariables(
    Particles.CoreSpotlightColor:LinearColor=&MultiplyLinearColor(
     Linear Color=<rgba 0.2 0.1687 0.16 1>
     Scale Factor=&MakeFloat4FromVectorAndFloat("Vector (RGB)"=<1 1 1>, "Float (Alpha)"=$System.Flicker)
    )
    Particles.FlareColor=&MultiplyLinearColor(Linear Color=$Particles.Color, Scale Factor=&MakeVector4_FromFloat(Float=0.333))
    Particles.FloorGlowColor:LinearColor=&MultiplyLinearColor(Linear Color=$Particles.Color, Scale Factor=&MakeVector4_FromFloat(Float=0.666))
    Particles.LightColor=&MultiplyLinearColor(Linear Color=$Particles.Color, Scale Factor=&MakeVector4_FromFloat(Float=8))
    Particles.LightComponentIntensity:NiagaraFloat=&Multiply_Float(A=$System.Oscillation, B=11111)
   ) {isSetParameters=true}
  ]
  rendererClasses [
   &NiagaraMeshRendererProperties
   &NiagaraLightRendererProperties
   &NiagaraComponentRendererProperties
   &NiagaraSpriteRendererProperties
  ]
  renderers [
   {rendererClass=&NiagaraMeshRendererProperties, rendererIndex=0}
   {rendererClass=&NiagaraMeshRendererProperties, rendererIndex=1}
   {rendererClass=&NiagaraLightRendererProperties, rendererIndex=2}
   {rendererClass=&NiagaraComponentRendererProperties, rendererIndex=3}
   {rendererClass=&NiagaraMeshRendererProperties, rendererIndex=4}
   {rendererClass=&NiagaraMeshRendererProperties, rendererIndex=5}
   {rendererClass=&NiagaraSpriteRendererProperties, rendererIndex=6}
   {rendererClass=&NiagaraSpriteRendererProperties, rendererIndex=7}
  ]
 }
 SpotLightSecondary {
  emitterUpdateScript [
   &EmitterState(
    Inactive Response=::NewEnumerator1
    Life Cycle Mode=::NewEnumerator1
    Loop Behavior=::NewEnumerator1
    Loop Duration Mode=::NewEnumerator1
    UseLoopDelay=false
   )
   &SpawnBurst_Instantaneous(Spawn Count=3, Use Spawn Probability=false, UseLoopCountLimit=false)
  ]
  particleSpawnScript [
   &InitializeParticle(
    Color=<rgba 0.3 0.271 0.24 1>
    Mesh Scale=<0.8 0.8 1.2>
    Mesh Scale Mode=::NewEnumerator4
    Sprite Size Mode=::NewEnumerator3
    Uniform Sprite Size=111
    UsePositionOffset=false
    Write Material Random=true
   )
   &CameraOffset()
   &PartitionParticles(Number Of Partitions=3)
  ]
  particleUpdateScript [
   &ParticleState(
    Kill Particles When Lifetime Has Elapsed=false
    Let Infinitely Lived Particles Die When Emitter Deactivates=true
    Loop Particles Lifetime=true
   )
   /SetVariables(Particles.Position=&SelectVectorFromArray(
     Direct Array Index=&Add_Integer(A=$Particles.PartitionParticles.Partition, B=&Multiply_Int(A=$System.Bar, B=3))
     Direct Array Mode=::NewEnumerator1
     Vector Selection Array=$User.SpotlightsPositionArray
    )) {isSetParameters=true, module=SetVariables_#8}
   &UpdateMeshOrientation(
    Facing Direction=&NormalizeVector(
     Vector To Normalize=&Subtract_Vector(
      A=&ConvertPositionToVector(Input Position=$Particles.Position, Conversion Operation:ENiagara_LWCConvertPosToVec=::NewEnumerator0)
      B=<0 0 111>
     )
     Safe Normalize=true
     Safe Fallback Threshold=0
     Safe Fallback Vector=<0 0 0.001>
    )
    Facing Interpolation=false
    Orientation Method=::NewEnumerator1
    Override Quat=false
    Reference Axis Vector:Vector3f=<0 0 1>
    Stabilization Mode:ENiagara_OrientVectorCount=::NewEnumerator0
    TransformFacing=false
   )
   &Color(
    Color=&SelectLinearColorFromArray(
     Array Sampling Mode=::NewEnumerator1
     Direct Array Index=&SelectIntFromArray(
      Direct Array Index=$System.Bar
      Direct Array Mode=::NewEnumerator1
      Int Selection Array=DataInterface() {propertyValues {IntData [3,2,4,0,1]}}
     )
     Direct Array Mode=::NewEnumerator1
     Color Selection Array=$User.SpotlightColorArray
     Random Seed=?
    )
    Scale Alpha=&Multiply_Float(
     A=$System.Flicker
     B=&SelectFloatFromArray(
      Direct Array Index=&Add_Integer(A=&Multiply_Int(A=$System.Beat, B=3), B=$Particles.PartitionParticles.Partition)
      Direct Array Mode=::NewEnumerator1
      Float Selection Array=DataInterface() {propertyValues {FloatData [0.6,1,0.4,1,0.4,0.6,0.4,0.6,1]}}
     )
    )
    Scale Color=<0.4 0.4 0.4>
   )
   /SetVariables(
    Particles.GlowCameraOffset=11
    Particles.GlowColor=&MultiplyLinearColor(Linear Color=$Particles.Color, Scale Factor=&MakeVector4_FromFloat(Float=0.6))
   ) {isSetParameters=true}
  ]
  rendererClasses [&NiagaraMeshRendererProperties, &NiagaraSpriteRendererProperties]
  renderers [
   {rendererClass=&NiagaraMeshRendererProperties, rendererIndex=0}
   {rendererClass=&NiagaraMeshRendererProperties, rendererIndex=1}
   {rendererClass=&NiagaraSpriteRendererProperties, rendererIndex=2}
  ]
 }
 SpotlightCircle {
  emitterUpdateScript [
   &EmitterState()
   &SpawnRate(
    Spawn Probability=&MakeCustomFloatFromBool(
     Bool=&Boolean_AndOperation(
      Bool1=&SetBoolByIntComparison(A=$System.BarModulo, B=2, Comparison Type=::NewEnumerator4)
      Bool2=&Boolean_AndOperation(
       Bool1=&SetBoolByIntComparison(A=$System.BeatModulo, B=0, Comparison Type=::NewEnumerator2)
       Bool2=&SetBoolByIntComparison(A=$System.BeatModulo, B=2, Comparison Type=::NewEnumerator7)
      )
     )
     True Float=1
     False Float=0
    )
    SpawnRate=6
    Use Spawn Probability=true
   )
  ]
  particleSpawnScript [
   &FN_SetSpawnGroup(Number of Spawn Groups=6)
   &InitializeParticle(
    Color Mode=::NewEnumerator0
    Lifetime Max=0.6
    Lifetime Min=0.8
    Lifetime Mode=::NewEnumerator1
    Mesh Scale=&Multiply_Vector(
     A=&SelectVectorFromArray(
      Direct Array Index=$Particles.SpawnGroup
      Vector Selection Array=DataInterface() {propertyValues {FloatData [<0.3 0.3 1.5>, <0.3 0.3 1.8>, <0.25 0.25 1.6>, <0.3 0.3 1.5>, <0.3 0.3 1.7>, <0.2 0.2 1.6>]}}
     )
     B=<1 1 0.8>
    )
    Mesh Scale Mode=::NewEnumerator4
    Position Offset=&Multiply_Vector(
     A=&SelectVectorFromArray(
      Direct Array Index=$Particles.SpawnGroup
      Vector Selection Array=DataInterface() {propertyValues {FloatData [<30 20 2>, <30 -20 2>, <0 -35 2>, <-40 -20 2>, <-40 20 2>, <0 35 2>]}}
     )
     B=<1 1 1>
    )
    Sprite Size Mode=::NewEnumerator3
    Uniform Sprite Size=77
    UsePositionOffset=true
    Write Material Random=true
   )
   &InitialMeshOrientation(
    Mesh Orientation Mode=::NewEnumerator7
    Rotation=&SelectVectorFromArray(
     Direct Array Index=$Particles.SpawnGroup
     Vector Selection Array=DataInterface() {propertyValues {FloatData [<0.2 0.5 -0.05>, <0.2 0.5 0.05>, <0 0.5 0.1>, <0 0.45 0.05>, <0 0.45 -0.05>, <0 0.5 -0.1>]}}
    )
    Use Rotation Vector=true
   )
   /SetVariables(Particles.GlowCameraOffset=33) {isSetParameters=true}
  ]
  particleUpdateScript [
   &ParticleState()
   &UpdateMeshOrientation(
    Rotation Coordinate Space=::NewEnumerator0
    Rotation Rate=&SelectFloatFromArray(Direct Array Index=$Particles.SpawnGroup, Float Selection Array=DataInterface() {propertyValues {FloatData [0.3,-0.4,0.35,-0.38,0.4,-0.35]}})
    Rotation Vector=<0 0 0.4>
   )
   &Color(
    Color=&SelectLinearColorFromArray(
     Array Sampling Mode=::NewEnumerator1
     Direct Array Index=$System.Bar
     Direct Array Mode=::NewEnumerator1
     Color Selection Array=$User.SpotlightColorArray
     Random Seed=?
    )
    Scale Color=<0.25 0.25 0.25>
   )
   &ScaleMeshSize(Scale Factor=&VectorFromCurve(VectorCurve=DataInterface() {propertyValues {XCurve {keys [<at 0 0>,<at 0.2 0.8>]}, YCurve {keys [<at 0 0>,<at 0.2 0.8>]}, ZCurve {keys [<at 0 1>,<at 1 1>]}}}))
   &ScaleSpriteSize(Uniform Curve Sprite Scale=DataInterface() {propertyValues {Curve {keys [{tangentMode=RCTM_User, time=0, value=0}, {tangentMode=RCTM_User, time=0.2, value=1}]}}})
   /SetVariables(Particles.GlowColor=&MultiplyLinearColor(Linear Color=$Particles.Color, Scale Factor=&MakeVector4_FromFloat(Float=0.5))) {isSetParameters=true}
  ]
  rendererClasses [&NiagaraMeshRendererProperties, &NiagaraSpriteRendererProperties]
  renderers [
   {rendererClass=&NiagaraMeshRendererProperties, rendererIndex=0}
   {rendererClass=&NiagaraSpriteRendererProperties, rendererIndex=1}
  ]
 }
}
summary {
 systemName=NS_Emote_Intertwine
 emitters [
  {
   emitterName=SpotlightCircle
   rendererClasses [&NiagaraMeshRendererProperties, &NiagaraSpriteRendererProperties]
  }
  {emitterName=SparkleRainParent, simTarget=GPUComputeSim, rendererClasses [&NiagaraSpriteRendererProperties]}
  {emitterName=SparkleRainTrail, simTarget=GPUComputeSim, rendererClasses [&NiagaraSpriteRendererProperties]}
  {emitterName=SoftSparkles, simTarget=GPUComputeSim, rendererClasses [&NiagaraSpriteRendererProperties]}
  {emitterName=FlashingSparkles, rendererClasses [&NiagaraSpriteRendererProperties]}
  {
   emitterName=SpotLightMain
   rendererClasses [
    &NiagaraMeshRendererProperties
    &NiagaraLightRendererProperties
    &NiagaraComponentRendererProperties
    &NiagaraSpriteRendererProperties
   ]
  }
  {emitterName=ConfettiStreamers, rendererClasses [&NiagaraMeshRendererProperties]}
  {emitterName=ConfettiSquares, rendererClasses [&NiagaraSpriteRendererProperties]}
  {
   emitterName=SpotLightSecondary
   rendererClasses [&NiagaraMeshRendererProperties, &NiagaraSpriteRendererProperties]
  }
  {emitterName=DustPuffLeftFoot, rendererClasses [&NiagaraSpriteRendererProperties]}
  {emitterName=DustPuffRightFoot, rendererClasses [&NiagaraSpriteRendererProperties]}
  {emitterName=SparkleRainCPULow, rendererClasses [&NiagaraSpriteRendererProperties]}
 ]
 userVariables [
  User.bIsFrontEnd {type=NiagaraBool}
  User.bIsFrontEndPreview {type=NiagaraBool}
  User.ConfettiColorArray {
   type=NiagaraDataInterfaceArrayColor
   description=""
   defaultValue {
    struct=NiagaraExt_VariableValue_DataInterface
    value {
     dataInterface=&self:NiagaraDataInterfaceArrayColor_1
     dataInterfaceClass=&NiagaraDataInterfaceArrayColor
    }
   }
  }
  User.FeetRestingPositions {
   type=NiagaraDataInterfaceArrayFloat3
   description=""
   defaultValue {
    struct=NiagaraExt_VariableValue_DataInterface
    value {
     dataInterface=&self:NiagaraDataInterfaceArrayFloat3_2
     dataInterfaceClass=&NiagaraDataInterfaceArrayFloat3
    }
   }
  }
  User.Local Alpha Preview {type=NiagaraFloat, defaultValue=0.75, description=""}
  User.Other Alpha Preview {type=NiagaraFloat, defaultValue=0.25, description=""}
  User.SongBPM {type=NiagaraFloat, defaultValue=97.5, description=""}
  User.SpotlightColorArray {
   type=NiagaraDataInterfaceArrayColor
   description=""
   defaultValue {
    struct=NiagaraExt_VariableValue_DataInterface
    value {
     dataInterface=&self:NiagaraDataInterfaceArrayColor_0
     dataInterfaceClass=&NiagaraDataInterfaceArrayColor
    }
   }
  }
  User.SpotlightsOrientationArray {
   type=NiagaraDataInterfaceArrayFloat3
   description=""
   defaultValue {
    struct=NiagaraExt_VariableValue_DataInterface
    value {
     dataInterface=&self:NiagaraDataInterfaceArrayFloat3_1
     dataInterfaceClass=&NiagaraDataInterfaceArrayFloat3
    }
   }
  }
  User.SpotlightsPositionArray {
   type=NiagaraDataInterfaceArrayFloat3
   description=""
   defaultValue {
    struct=NiagaraExt_VariableValue_DataInterface
    value {
     dataInterface=&self:NiagaraDataInterfaceArrayFloat3_0
     dataInterfaceClass=&NiagaraDataInterfaceArrayFloat3
    }
   }
  }
 ]
}
